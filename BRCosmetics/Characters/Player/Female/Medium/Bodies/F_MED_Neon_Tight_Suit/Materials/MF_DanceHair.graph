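# Material function for the DanceHair effect on the female Neon Tight Suit character, making hair cycle between two glowing colors.
# A time-driven remapped sine, clamped to hold at each end, blends HairColorA and HairColorB, shaded by a camera-facing normal-dot-view falloff.
# The result replaces BaseColor and, scaled, Emissive on the incoming material attributes.
# Metallic is overridden outright, while Specular and Roughness blend toward override parameters through mix weights.
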MaterialFunction /BRCosmetics/Characters/Player/Female/Medium/Bodies/F_MED_Neon_Tight_Suit/Materials/MF_DanceHair
graph {nodes [
  Constant() {r=0.5}
  Fresnel() {baseReflectFraction=0, exponent=0.1}
  n1 = GetMaterialAttributes(None=FunctionInput() {inputType=FunctionInput_MaterialAttributes}) {attributeGetTypes [#1, #2]}
  n2 = Clamp(None=DotProduct(A=PixelNormalWS(), B=CameraVectorWS()))
  n3 = ScalarParameter() {defaultValue=2, group=DanceHair, parameterName=ViewDepedantFalloff}
  n4 = MaterialFunctionCall!(
   Sine Phase=Multiply(A=Time(), B=ScalarParameter() {defaultValue=0.05, parameterName=HairSineSpeed})
   Value 1=Constant() {r=-1.5}
   Value 2=Constant() {r=2.5}
  ) {materialFunction=/Engine/Functions/Engine_MaterialFunctions02/Utility/Sine_Remapped}
  n5 = Multiply(
   A=LinearInterpolate(B=Power(Base=@n2, Exp=@n3), Alpha=ScalarParameter() {defaultValue=1, group=DanceHair, parameterName=ViewDependantStrength}) {constA=1}
   B=LinearInterpolate(
    A=VectorParameter() {group=DanceHair, parameterName=HairColorA, defaultValue=<rgba 1 0.102 0.9098 1>}.RGB
    B=VectorParameter() {group=DanceHair, parameterName=HairColorB, defaultValue=<rgba 0.1804 0.8196 1 1>}.RGB
    Alpha=Clamp(None=@n4.Remapped Sine)
   )
  )
  n6 = SetMaterialAttributes!(
   MaterialAttributes=@n1.MaterialAttributes
   BaseColor=@n5
   EmissiveColor=Multiply(A=@n5, B=ScalarParameter() {defaultValue=0.3, group=DanceHair, parameterName=emissiveMultiplier}) {constB=0.3}
   Specular=LinearInterpolate(
    A=@n1.Specular
    B=VectorParameter() {group=DanceHair, parameterName=SpecularOverride, defaultValue=<rgba 0.4735 0.4735 0.4678 1>}.RGB
    Alpha=ScalarParameter() {group=DanceHair, parameterName=SpecularMix}
   )
   Metallic=VectorParameter() {group=DanceHair, parameterName=Metallic, defaultValue=<rgba 0.3231 0.3231 0.3231 1>}.RGB
   Roughness=LinearInterpolate(
    A=@n1.Roughness
    B=ScalarParameter() {defaultValue=0.5, group=DanceHair, parameterName=Roughness}
    Alpha=ScalarParameter() {group=DanceHair, parameterName=RoughnessMix}
   )
  ) {attributeSetTypes [
    #3
    #4
    #1
    #5
    #2
   ]}
  FunctionOutput(None=@n6)
 ]}
