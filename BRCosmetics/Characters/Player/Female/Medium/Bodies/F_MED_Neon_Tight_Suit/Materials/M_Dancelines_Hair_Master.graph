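# M_Dancelines_Hair_Master is a Default Lit master material for hair on the female Neon Tight Suit character, set up for skeletal meshes and mesh deformers.
# It uses material attributes, passing the shared character PreFX result through MF_DanceHair into the shared PostFX function.
# The PreFX AO output alone drives PostFX's AO, Distance Blend, Fresnel and Gmap inputs.
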Material /BRCosmetics/Characters/Player/Female/Medium/Bodies/F_MED_Neon_Tight_Suit/Materials/M_Dancelines_Hair_Master
graph {
 properties {
  bUseMaterialAttributes=true
  bUsedWithMeshDeformer=true
  bUsedWithSkeletalMesh=true
  pixelDepthOffsetMode=PDOM_Legacy
  usedShadingModels=Default Lit
 }
 nodes [
  n1 = MaterialFunctionCall() {materialFunction=&BaseFN_CharacterFunction_PreFX}
  MaterialFunctionCall(
   AO=@n1.AO
   Distance Blend=@n1.AO
   Fresnel=@n1.AO
   Gmap=@n1.AO
   Material=MaterialFunctionCall!(In=@n1.Result) {materialFunction=&/MF_DanceHair}.Result
  ) {materialFunction=&BaseFN_CharacterFunction_PostFX}
 ]
}
